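# UI sound cue for the front-end's default common X/close button press, routed to the Fort_UI sound class.
# A Mixer layers the main button click with a small whoosh, which is weighted higher at the mixer input but first passes through a Modulator that lowers its volume by a fixed amount.
# Playback references a shared front-end menu concurrency set that, by its name, caps simultaneous sounds and stops the oldest.
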SoundCue /Game/Sounds/Fort_UI_Sounds/FrontEnd/DefaultSounds/PressedButtons/FN_UI_Default_Common_X_Close_Cue
root=Mixer_0
nodes [
 {id=Mixer_0, children [WavePlayer_1,Modulator_1], properties {inputVolume [1,1.5]}}
 {id=Modulator_1, children [WavePlayer_0], properties {volumeMax=0.5, volumeMin=0.5}}
 {
  id=WavePlayer_0
  properties {soundWaveAssetPtr=/Game/Sounds/Fort_UI_Sounds/FrontEnd/_GenericToolkit_/Whooshes/Misc_Whooshes/wav/UI_Toolkit_Whoosh_Sm_02}
 }
 {id=WavePlayer_1, properties {soundWaveAssetPtr=/Game/Sounds/Fort_UI_Sounds/FrontEnd/New_FrontEnd/UI_FN_Button_Click_Main_01}}
]
properties {
 maxDistance=2097152
 soundClassObject=&Fort_UI
 volumeMultiplier=1
 concurrencyOverrides {maxCount=1, resolutionRule=StopOldest}
 concurrencySet [/Game/Sounds/Concurrency_Rules/FrontEnd_UI_Menu_Max_03_Stop_Oldest]
}
