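# One-shot stereo UI MetaSound for pressing a locked front-end button, playing a deny sound in the Fort_UI sound class.
# OnPlay triggers two MSS_PlayRandomSound_Stereo instances: one fed by the Sounds input, the other by graph variables with no values shown.
# A stereo mixer blends both at slightly reduced gain; only the first instance signals OnFinished.
# Concurrency allows one instance, stopping the oldest.
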MetaSoundSource /Game/Sounds/Fort_UI_Sounds/FrontEnd/DefaultSounds/PressedButtons/MSS_UI_Button_Pressed_Locked
inputs [
 Sounds {type=WaveAsset:Array}
 UE.Source.OnPlay {type=Trigger}
]
interfaces [UE.OutputFormat.Stereo:1.0, UE.Source.OneShot:1.0, UE.Source:1.1]
outputs [
 "UE.OutputFormat.Stereo.Audio:0" {type=Audio}
 "UE.OutputFormat.Stereo.Audio:1" {type=Audio}
 UE.Source.OneShot.OnFinished {type=Trigger}
]
pages [Default {
  input_defaults {UE.Source.OnPlay=null, Sounds [/Game/Sounds/Fort_UI_Sounds/FrontEnd/DefaultSounds/PressedButtons/wav/UI_Deny_Play_Button_Locked_E]}
  nodes [
   MSS_PlayRandomSound_Stereo_0 = &MSS_PlayRandomSound_Stereo(Sounds=^Sounds, UE.Source.OnPlay=^UE.Source.OnPlay)
   n1 = &MSS_PlayRandomSound_Stereo(
    UE.Source.OnPlay=^UE.Source.OnPlay
    MaxVolume=VariableAccessor.Float(Variable=InitVariable.Float().Variable).Value
    Sounds=VariableAccessor.WaveAsset:Array(Variable=InitVariable.WaveAsset:Array().Variable).Value
   )
   AudioMixerAudioMixerStereo2_0 = "AudioMixer.Audio Mixer (Stereo, 2)"(
    In 0 L=@MSS_PlayRandomSound_Stereo_0."UE.OutputFormat.Stereo.Audio:0"
    In 0 R=@MSS_PlayRandomSound_Stereo_0."UE.OutputFormat.Stereo.Audio:1"
    In 1 L=@n1."UE.OutputFormat.Stereo.Audio:0"
    In 1 R=@n1."UE.OutputFormat.Stereo.Audio:1"
   ) {also {defaults {Gain 0=0.8, Gain 1=0.8}}}
  ]
  links [
   {dst="@UE.OutputFormat.Stereo.Audio:0", src=AudioMixerAudioMixerStereo2_0.Out L}
   {dst="@UE.OutputFormat.Stereo.Audio:1", src=AudioMixerAudioMixerStereo2_0.Out R}
   {dst="@UE.Source.OneShot.OnFinished", src=MSS_PlayRandomSound_Stereo_0.UE.Source.OneShot.OnFinished}
  ]
 }]
properties {
 assetImportData=&self:AssetImportData
 bOverrideConcurrency=true
 numChannels=2
 outputFormat=Stereo
 sampleRate=48000
 soundClassObject=&Fort_UI
 virtualizationMode=Disabled
 volume=1.007
 concurrencyOverrides {maxCount=1, resolutionRule=StopOldest}
}
