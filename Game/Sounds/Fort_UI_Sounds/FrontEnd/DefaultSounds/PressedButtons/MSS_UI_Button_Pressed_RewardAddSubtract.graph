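# One-shot stereo MetaSound for the UI button press, named for reward add/subtract, playing a front-end button click wave.
# The integer Pitch input converts to float and drives the wave player's pitch shift.
# The custom Play trigger, not OnPlay, starts playback; the Min/Max pitch and volume inputs are unconnected.
# Pressed-button concurrency is assigned, with an override limiting it to one voice, stopping the oldest.
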MetaSoundSource /Game/Sounds/Fort_UI_Sounds/FrontEnd/DefaultSounds/PressedButtons/MSS_UI_Button_Pressed_RewardAddSubtract
inputs [
 MaxPitch {type=Float}
 MaxVolume {type=Float}
 MinPitch {type=Float}
 MinVolume {type=Float}
 Pitch {type=Int32}
 Play {type=Trigger}
 UE.Source.OnPlay {type=Trigger}
]
interfaces [UE.OutputFormat.Stereo:1.0, UE.Source.OneShot:1.0, UE.Source:1.1]
outputs [
 "UE.OutputFormat.Stereo.Audio:0" {type=Audio}
 "UE.OutputFormat.Stereo.Audio:1" {type=Audio}
 UE.Source.OneShot.OnFinished {type=Trigger}
]
pages [Default {
  input_defaults {MaxPitch=1, MaxVolume=1, MinPitch=1, MinVolume=1, Pitch=null, Play=null, UE.Source.OnPlay=null}
  nodes [
   InitVariable.Float()
   WavePlayer_0 = UE.Wave Player:Stereo(Play=^Play, Pitch Shift=Convert.Float:Int32(Int32=^Pitch).Float) {also {defaults {Wave Asset=/Game/Sounds/Fort_UI_Sounds/FrontEnd/New_FrontEnd/UI_FN_Button_Click_Main_01}}}
  ]
  links [
   {dst="@UE.Source.OneShot.OnFinished", src=WavePlayer_0.On Finished}
   {dst="@UE.OutputFormat.Stereo.Audio:0", src=WavePlayer_0.Out Left}
   {dst="@UE.OutputFormat.Stereo.Audio:1", src=WavePlayer_0.Out Right}
  ]
 }]
properties {
 assetImportData=&self:AssetImportData
 numChannels=2
 outputFormat=Stereo
 sampleRate=48000
 soundClassObject=&Fort_UI
 virtualizationMode=Disabled
 concurrencyOverrides {maxCount=1, resolutionRule=StopOldest}
 concurrencySet [&/UI_PressedButton_Concurrency]
}
